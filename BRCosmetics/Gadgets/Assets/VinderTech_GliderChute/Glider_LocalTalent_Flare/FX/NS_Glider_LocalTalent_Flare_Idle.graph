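# Niagara idle effect for the Local Talent Flare glider, with trails, sparkles, tumbling sprites and speed lines.
# MainTrail002 and SoftTrail draw paired socket-based ribbons, while Sprites and GPU Trail_Sparkles follow two skeletal mesh sockets and drift outward.
# Two smoothed system values scale velocity, spread and opacity, and ForwardDot drives sprite spawning and a trail texture parameter.
# VariantSwitch selects color variants, front-end preview overrides motion and intensity, and time of day dims some layers.
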NiagaraSystem /BRCosmetics/Gadgets/Assets/VinderTech_GliderChute/Glider_LocalTalent_Flare/FX/NS_Glider_LocalTalent_Flare_Idle
dependencies {
 usedDataInterfaces [
  &NiagaraDataInterfaceArrayInt32
  &NiagaraDataInterfaceArrayFloat3
  &NiagaraDataInterfaceCurve
  &NiagaraDataInterfaceVector2DCurve
  &NiagaraDataInterfaceColorCurve
  &NiagaraDataInterfaceArrayColor
  &NiagaraDataInterfaceArrayFloat
  &NiagaraDataInterfacePlatformSet
  &NiagaraDataInterfaceSocketReader
  &NiagaraDataInterfaceSkeletalMesh
 ]
 usedDynamicInputs []
 usedModules [
  &self:SystemSpawnScript.SystemScriptSource.SystemScriptGraph.NiagaraNodeAssignment_0.SetVariables_#1
  &self:SystemSpawnScript.SystemScriptSource.SystemScriptGraph.NiagaraNodeAssignment_1.SetVariables_#2
  &SystemState
  &InterpolateOverTime
 ]
}
emitters {
 MainTrail002 {
  emitterUpdateScript [
   &EmitterState(Life Cycle Mode=::NewEnumerator1, Loop Duration=1, UseLoopDelay=false)
   &FN_SpawnAnimTrail(Number of Anim Trails=2)
   /SetVariables(Emitter.TrailTex_001U:NiagaraFloat=&Lerp_Float(
     A=0.2
     B=0.3
     Alpha=&MakeCustomFloatFromBool(Bool=&SetBoolByFloatComparison(A=$User.ForwardDot, B=0.2), True Float=1, False Float=0)
    )) {isSetParameters=true}
   /SetVariables(Emitter.LinearMaskLeft_Power:NiagaraFloat=&FN_PickFloatDependOnPlatform(TargetPlatformActive=1, TargetPlatformNoActive=3)) {isSetParameters=true}
  ]
  particleSpawnScript [
   &FN_InitializeAnimTrailBoneIndecies(First Bone or Socket Index=1, Second Bone or Socket Index=0)
   &FN_InitializeAnimTrailBoneIndecies(AnimTrail Selection=1, First Bone or Socket Index=3, Second Bone or Socket Index=2) {module=FN_InitializeAnimTrailBoneIndecies001}
   &FN_InitializeAnimTrail(AnimTrailWidth=1.2)
   &AddVelocity(Velocity=&MakeCustomVectorFromBool(
     Bool=$User.bIsFrontEndPreview
     True Vector=<0 -500 0>
     False Vector=&Lerp_Vector(
      A=&Lerp_Vector(A=<0 300 0>, B=<0 400 0>, Alpha=$System.InterpolateOverTime001.InterpolateOverTime_Float.Moving Average)
      B=<0 450 0>
      Alpha=$System.InterpolateOverTime.InterpolateOverTime_Float.Moving Average
     )
    ))
  ]
  particleUpdateScript [
   &FN_UpdateAnimTrail()
   &ScaleRibbonWidth(Ribbon Width Scale=&MakeCustomFloatFromBool(
     Bool=$User.bIsFrontEndPreview
     True Float=1.2
     False Float=&Lerp_Float(
      A=&Lerp_Float(A=1, B=1, Alpha=$System.InterpolateOverTime001.InterpolateOverTime_Float.Moving Average)
      B=1
      Alpha=$System.InterpolateOverTime.InterpolateOverTime_Float.Moving Average
     )
    ))
   &ScaleColor(
    Scale Alpha=&FloatFromCurve(
     FloatCurve=DataInterface() {propertyValues {Curve {keys [
         <at 0 1>
         {arriveTangent=-3.1841, interpMode=RCIM_Cubic, leaveTangent=-3.1841, tangentMode=RCTM_User, time=0.5, value=1}
         <at 1 0 RCIM_Cubic>
        ]}}}
     Scale Curve=&FN_FrontEndFloatModifier(Gameplay=0.8, Front End=0.4)
    )
    Scale RGB=&MakeVectorFromLinearColorRGB(Color=&SelectLinearColorFromArray(
      Array Sampling Mode=::NewEnumerator1
      Direct Array Index=&MakeIntFromFloat(Float=$User.VariantSwitch)
      Color Selection Array=DataInterface() {propertyValues {ColorData [<rgba 1 1 1 1>, <rgba 0.6308 1 0 1>]}}
      Random Seed=?
     ))
    ScaleA=true
    ScaleRGB=true
   )
   /SetVariables(Particles.ColorPoint:LinearColor=&MakeLinearColorFromVectorAndFloat(
     "Vector (RGB)"=&Multiply_VectorByFloat(
      Vector=&MakeVectorFromLinearColorRGB(Color=&SelectLinearColorFromArray(
        Array Sampling Mode=::NewEnumerator1
        Direct Array Index=&MakeIntFromFloat(Float=$User.VariantSwitch)
        Color Selection Array=DataInterface() {propertyValues {ColorData [<rgba 1 1 1 1>, <rgba 0.9742 1 0.0365 1>]}}
        Random Seed=?
       ))
      Float=1
     )
     "Float (Alpha)"=&Multiply_Float(
      A=&FloatFromCurve(
       FloatCurve=DataInterface() {propertyValues {Curve {keys [<at 0 1>, {arriveTangent=-3.094, interpMode=RCIM_Cubic, leaveTangent=-3.094, tangentMode=RCTM_User, time=0.5, value=1}, <at 1 0 RCIM_Cubic>]}}}
       Scale Curve=&FN_FrontEndFloatModifier(Gameplay=1.5, Front End=1.2)
      )
      B=1
     )
    )) {isSetParameters=true}
   /SetVariables(Particles.ColorLine_Green:LinearColor=&MakeLinearColorFromVectorAndFloat(
     "Vector (RGB)"=&MakeVectorFromLinearColorRGB(Color=&Multiply_LinearColorByFloat(
       Linear Color=&SelectLinearColorFromArray(
        Array Sampling Mode=::NewEnumerator1
        Direct Array Index=&MakeIntFromFloat(Float=$User.VariantSwitch)
        Color Selection Array=DataInterface() {propertyValues {ColorData [<rgba 0 0.33 0.2318 1>, <rgba 0.1774 0.9 0 1>]}}
        Random Seed=?
       )
       Float=&MakeCustomFloatFromBool(Bool=$User.bIsFrontEndPreview, True Float=0.7, False Float=1)
      ))
     "Float (Alpha)"=&FloatFromCurve(
      FloatCurve=DataInterface() {propertyValues {Curve {keys [{arriveTangent=-1.4846, interpMode=RCIM_Cubic, leaveTangent=-1.4846, tangentMode=RCTM_User, time=0, value=1}, <at 1 0 RCIM_Cubic>]}}}
      Scale Curve=&FN_FrontEndFloatModifier(Gameplay=3, Front End=2)
     )
    )) {isSetParameters=true}
   /SetVariables(Particles.ColorLine_Yellow:LinearColor=&MakeLinearColorFromVectorAndFloat(
     "Vector (RGB)"=&MakeVectorFromLinearColorRGB(Color=&SelectLinearColorFromArray(
       Array Sampling Mode=::NewEnumerator1
       Direct Array Index=&MakeIntFromFloat(Float=$User.VariantSwitch)
       Color Selection Array=DataInterface() {propertyValues {ColorData [<rgba 0.996 1 0 1>, <rgba 0.0356 0.3292 3 1>]}}
       Random Seed=?
      ))
     "Float (Alpha)"=&Multiply_Float(
      A=&Multiply_Float(
       A=&FloatFromCurve(
        FloatCurve=DataInterface() {propertyValues {Curve {keys [{arriveTangent=-1.436, interpMode=RCIM_Cubic, leaveTangent=-1.436, tangentMode=RCTM_User, time=0, value=1}, <at 1 0 RCIM_Cubic>]}}}
        Scale Curve=&FN_FrontEndFloatModifier(Gameplay=3, Front End=2)
       )
       B=1
      )
      B=1
     )
    )) {isSetParameters=true}
   &SolveForcesAndVelocity(Clamp Velocity=false, Limit Acceleration=false)
   &DynamicMaterialParameters(
    Index 2 Param 1=1
    Index 2 Param 3=1
    Index 2 Param 4=&SelectFloatFromArray(
     Direct Array Index=&MakeIntFromFloat(Float=$User.VariantSwitch)
     Float Selection Array=DataInterface() {propertyValues {FloatData [0,1]}}
    )
    Module.Index2Float0Write=true
    Module.Index2Float1Write=true
    Module.Index2Float2Write=true
    Module.Index2Float3Write=true
    Write Parameter Index 0=::NewEnumerator0
    Write Parameter Index 2=::NewEnumerator1
   )
  ]
  rendererClasses [&NiagaraRibbonRendererProperties]
  renderers [
   {rendererClass=&NiagaraRibbonRendererProperties}
   {rendererClass=&NiagaraRibbonRendererProperties, rendererIndex=1}
   {rendererClass=&NiagaraRibbonRendererProperties, rendererIndex=2}
   {rendererClass=&NiagaraRibbonRendererProperties, rendererIndex=3}
  ]
 }
 SoftTrail {
  emitterUpdateScript [
   &EmitterState(Life Cycle Mode=::NewEnumerator1, Loop Duration=0.1, UseLoopDelay=false)
   &FN_SpawnAnimTrail(Number of Anim Trails=2)
  ]
  particleSpawnScript [
   &FN_InitializeAnimTrailBoneIndecies(First Bone or Socket Index=1, Second Bone or Socket Index=0)
   &FN_InitializeAnimTrailBoneIndecies(AnimTrail Selection=1, First Bone or Socket Index=3, Second Bone or Socket Index=2) {module=FN_InitializeAnimTrailBoneIndecies001}
   &FN_InitializeAnimTrail(AnimTrailWidth=1.5)
   &AddVelocity(Velocity=&MakeCustomVectorFromBool(
     Bool=$User.bIsFrontEndPreview
     True Vector=<0 -600 0>
     False Vector=&Lerp_Vector(
      A=&Lerp_Vector(A=<0 400 0>, B=<1 500 0>, Alpha=$System.InterpolateOverTime001.InterpolateOverTime_Float.Moving Average)
      B=<0 550 0>
      Alpha=$System.InterpolateOverTime.InterpolateOverTime_Float.Moving Average
     )
    ))
  ]
  particleUpdateScript [
   &FN_UpdateAnimTrail()
   &ScaleRibbonWidth(Ribbon Width Scale=&MakeCustomFloatFromBool(
     Bool=$User.bIsFrontEndPreview
     True Float=1.1
     False Float=&Lerp_Float(
      A=&Lerp_Float(A=0.8, B=1.1, Alpha=$System.InterpolateOverTime001.InterpolateOverTime_Float.Moving Average)
      B=1.2
      Alpha=$System.InterpolateOverTime.InterpolateOverTime_Float.Moving Average
     )
    ))
   &ScaleColor(
    Scale Alpha=&Multiply_Float(
     A=&FloatFromCurve(
      FloatCurve=DataInterface() {propertyValues {Curve {keys [{arriveTangent=-1.753, interpMode=RCIM_Cubic, leaveTangent=-1.753, tangentMode=RCTM_User, time=0, value=1}, <at 1 0 RCIM_Cubic>]}}}
      Scale Curve=&Multiply_Float(
       A=&FN_FrontEndFloatModifier(Gameplay=4.5, Front End=2)
       B=&FNTimeOfDayScalar(Night Scale=0.7, Morning Scale=0.8, Evening Scale=0.8)
      )
     )
     B=&MakeCustomFloatFromBool(
      Bool=$User.bIsFrontEndPreview
      True Float=1
      False Float=&Lerp_Float(
       A=&Lerp_Float(A=0, B=1, Alpha=$System.InterpolateOverTime001.InterpolateOverTime_Float.Moving Average)
       B=1
       Alpha=$System.InterpolateOverTime.InterpolateOverTime_Float.Moving Average
      )
     )
    )
    Scale RGB=&MakeVectorFromLinearColorRGB(Color=&SelectLinearColorFromArray(
      Array Sampling Mode=::NewEnumerator1
      Direct Array Index=&MakeIntFromFloat(Float=$User.VariantSwitch)
      Color Selection Array=DataInterface() {propertyValues {ColorData [<rgba 0.0687 0.5 0.1018 1>, <rgba 0.1119 0.5 0.068 1>]}}
      Random Seed=?
     ))
    ScaleA=true
    ScaleRGB=true
   )
   &SolveForcesAndVelocity(Clamp Velocity=false, Limit Acceleration=false)
   &DynamicMaterialParameters(
    Index 2 Param 1=1
    Index 2 Param 3=1
    Index 2 Param 4=&SelectFloatFromArray(
     Direct Array Index=&MakeIntFromFloat(Float=$User.VariantSwitch)
     Float Selection Array=DataInterface() {propertyValues {FloatData [0,1]}}
    )
    Module.Index2Float0Write=true
    Module.Index2Float1Write=true
    Module.Index2Float2Write=true
    Module.Index2Float3Write=true
    Write Parameter Index 0=::NewEnumerator0
    Write Parameter Index 2=::NewEnumerator1
   )
  ]
  rendererClasses [&NiagaraRibbonRendererProperties]
  renderers [{rendererClass=&NiagaraRibbonRendererProperties}]
 }
 SpeedLine {
  simTarget=GPUComputeSim
  emitterUpdateScript [
   &EmitterState(
    Delay First Loop Only=true
    Inactive Response=::NewEnumerator1
    Life Cycle Mode=::NewEnumerator1
    Loop Delay=1
    Loop Duration=1
    UseLoopDelay=true
   )
   &SpawnRate(
    SpawnRate=&MakeCustomFloatFromBool(Bool=$User.bIsFrontEndPreview, True Float=200, False Float=150)
    Use Spawn Probability=true
   ) {module=SpawnRate001}
  ]
  particleSpawnScript [
   &InitializeParticle(
    Color=&MakeLinearColorFromVectorAndFloat(
     "Vector (RGB)"=<1 1 1>
     "Float (Alpha)"=&Multiply_Float(
      A=&UniformRangedFloat(Minimum=0.1, Maximum=0.2)
      B=&MakeCustomFloatFromBool(Bool=$User.bIsFrontEndPreview, True Float=0.5, False Float=1)
     )
    )
    Lifetime Max=0.1
    Lifetime Min=0.1
    Lifetime Mode=::NewEnumerator1
    Sprite Size Max=<2 15>
    Sprite Size Min=<2 5>
    Sprite Size Mode=::NewEnumerator2
    UsePositionOffset=false
    Write Material Random=true
   )
   &ShapeLocation(
    Handle Radius=0
    Large Radius=200
    MaskSpawnGroup=false
    Offset=&MakeCustomVectorFromBool(
     Bool=$User.bIsFrontEndPreview
     True Vector=&RandomRangeVector(Minimum=<0 300 0>, Maximum=<0 0 0>)
     False Vector=&RandomRangeVector(Minimum=<0 0 0>, Maximum=<0 -300 0>)
    )
    Offset Mode=::NewEnumerator0
    Override Seed=false
    Rotation Axis=<0 1 0>
    Rotation Mode=::NewEnumerator1
    Shape Origin=$Engine.Emitter.SimulationPosition
    Shape Primitive=::NewEnumerator3
    UseOwnerScale=true
   )
   &AddVelocity(
    Velocity=<0 500 0>
    Velocity Speed Scale=&MakeCustomFloatFromBool(Bool=$User.bIsFrontEndPreview, True Float=-1.5, False Float=1)
   ) {module=AddVelocity001}
   &AddVelocity(Velocity=&Multiply_Vector(
     A=<1 1 1>
     B=&MakeVector(X=1, Y=&MakeCustomFloatFromBool(Bool=$User.bIsFrontEndPreview, True Float=-1, False Float=1), Z=1)
    )) {enabled=false}
  ]
  particleUpdateScript [
   &ParticleState()
   &ScaleSpriteSize(Scale Factor=&Vector2DFromCurve(
     Vector2Curve=DataInterface() {propertyValues {XCurve {keys [<at 0 0.5 RCIM_Cubic>,<at 0.5 1 RCIM_Cubic>]}, YCurve {keys [<at 0 0 RCIM_Cubic>,<at 0.5 1 RCIM_Cubic>]}}}
     ScaleCurve=&Vector2DFromFloat(Value=1)
    ))
   &SolveForcesAndVelocity(Clamp Velocity=false, Limit Acceleration=false)
   &ScaleSpriteSizeBySpeed(Max Scale Factor=<1 30>, Min Scale Factor=<0.01 0.05>, Velocity Threshold=1)
   &ScaleColor(
    Scale Alpha=&FloatFromCurve(
     FloatCurve=DataInterface() {propertyValues {Curve {keys [
         <at 0 0 RCIM_Cubic>
         <at 0.2 1.0207>
         {arriveTangent=-2.221, interpMode=RCIM_Cubic, leaveTangent=-2.221, tangentMode=RCTM_User, time=1, value=0}
        ]}}}
     Scale Curve=&MakeCustomFloatFromBool(
      Bool=$User.bIsFrontEndPreview
      True Float=&MakeCustomFloatFromBool(Bool=$User.bIsFullyDeployed, True Float=0.5, False Float=0)
      False Float=&Lerp_Float(A=0, B=1, Alpha=$System.InterpolateOverTime.InterpolateOverTime_Float.Moving Average)
     )
    )
    Scale RGB=&VectorFromFloat(Value=3)
    ScaleA=true
    ScaleRGB=true
   )
  ]
  rendererClasses [&NiagaraSpriteRendererProperties]
  renderers [{rendererClass=&NiagaraSpriteRendererProperties}]
 }
 Sprites {
  emitterUpdateScript [
   &EmitterState(
    Enable Distance Culling=true
    Inactive Response=::NewEnumerator1
    Life Cycle Mode=::NewEnumerator1
    Loop Duration=1
    Max Distance Response=::KillImmediately
    Module.bUseMaxDistance=true
    Module.bUseMinDistance=false
    Scalability Mode=::NewEnumerator1
    UseLoopDelay=false
   )
   &SpawnRate(
    Spawn Probability=&MakeCustomFloatFromBool(Bool=$User.bIsFrontEndPreview, True Float=0.7, False Float=&Lerp_Float(A=0, B=1, Alpha=$User.ForwardDot))
    SpawnRate=60
    Use Spawn Probability=true
   )
  ]
  particleSpawnScript [
   &PartitionParticles()
   &InitializeParticle(
    Color=&MakeCustomLinearColorFromBool(
     Bool=&RandomBool()
     True Color=<rgba 0 0.65 0.2256 1>
     False Color=&MakeCustomLinearColorFromBool(
      Bool=&RandomBool()
      True Color=<rgba 0.7658 1 0 1>
      False Color=&LerpLinearColors(StartColor=<rgba 0.6979 0.1713 0 1>, EndColor=<rgba 0.6979 0.103 0.6872 1>, LerpFactor=$User.VariantSwitch)
     )
    )
    Lifetime Max=1
    Lifetime Min=0.5
    Lifetime Mode=::NewEnumerator1
    Sprite Rotation Mode=::NewEnumerator2
    Sprite Size Mode=::NewEnumerator1
    Uniform Sprite Size Max=20
    Uniform Sprite Size Min=10
    UsePositionOffset=true
    Write Material Random=true
   )
   /SetVariables(Particles.VisibilityTag=&RandomRangeInt(Minimum=0, Maximum=3)) {isSetParameters=true}
   &ShapeLocation(
    Cylinder Height=130
    Cylinder Radius=12
    MaskSpawnGroup=false
    Non Uniform Scale=<1 1 0.5>
    Offset Mode=::NewEnumerator0
    Override Seed=false
    Rotation Axis=<1 0.5 0.4>
    Rotation Mode=::NewEnumerator1
    Shape Primitive=::NewEnumerator1
    Surface Only=false
    Surface Only Band Thickness=0.1
    UseOwnerScale=true
   )
  ]
  particleUpdateScript [
   &ParticleState()
   &SkeletalMeshLocation(
    Bone / Socket Index=&SelectIntFromArray(Direct Array Index=$Particles.PartitionParticles.Partition, Int Selection Array=DataInterface() {propertyValues {IntData [0,1]}})
    MaskSpawnGroup=false
    Mesh Sampling Type=::NewEnumerator5
    Sampled Position Offset=&AddVectorToPosition(
     Position=Linked() {linkedVariable {name=Particles.Initial.ShapeLocation.ShapeVector, type=Vector3f}}
     Vector=&Lerp_Vector(
      A=&SelectVectorFromArray(
       Direct Array Index=$Particles.PartitionParticles.Partition
       Vector Selection Array=DataInterface() {propertyValues {FloatData [<40 20 15>, <-40 -20 -15>]}}
      )
      B=&SelectVectorFromArray(
       Direct Array Index=$Particles.PartitionParticles.Partition
       Vector Selection Array=DataInterface() {propertyValues {FloatData [<40 100 15>, <-40 -100 -15>]}}
      )
      Alpha=&FloatFromCurve(
       FloatCurve=DataInterface() {propertyValues {Curve {keys [<at 0 0>,<at 1 1>]}}}
       Scale Curve=&MakeCustomFloatFromBool(
        Bool=$User.bIsFrontEndPreview
        True Float=2.5
        False Float=&Lerp_Float(
         A=&Lerp_Float(A=1, B=3, Alpha=$System.InterpolateOverTime001.InterpolateOverTime_Float.Moving Average)
         B=4
         Alpha=$System.InterpolateOverTime.InterpolateOverTime_Float.Moving Average
        )
       )
      )
     )
    )
    Skeletal Mesh=$System.Skeletal Mesh Reader
    Socket Sampling Mode=::NewEnumerator5
   )
   &SpriteRotationRate(Rotation Rate=&RandomRangeFloat(Minimum=-200, Maximum=200))
   &ScaleSpriteSize(Scale Factor=&Vector2DFromFloat(Value=&FloatFromCurve(FloatCurve=DataInterface() {propertyValues {Curve {keys [
          {interpMode=RCIM_Cubic, tangentMode=RCTM_User, time=0, value=1}
          {arriveTangent=-0.0231, interpMode=RCIM_Cubic, leaveTangent=-0.0231, tangentMode=RCTM_User, time=0.7, value=1}
          {tangentMode=RCTM_User, time=1, value=0}
         ]}}})))
   /SetVariables(Particles.Size002:Vector2f=&Multiply_Vector2D(A=$Particles.SpriteSize, B=&Vector2DFromFloat(Value=1.5))) {isSetParameters=true}
   /SetVariables(Particles.Size_Triangle:Vector2f=&Multiply_Vector2D(A=$Particles.SpriteSize, B=&Vector2DFromFloat(Value=0.7))) {isSetParameters=true}
   &Drag(Drag=&RandomRangeFloat(Minimum=1, Maximum=2), Use Linear Drag=true, Use Rotational Drag=true)
   &SolveForcesAndVelocity(Clamp Velocity=false, Limit Acceleration=false)
   &ScaleColor(
    Scale Alpha=&Multiply_Float(
     A=&FNTimeOfDayScalar(Night Scale=1, Morning Scale=0.3, Evening Scale=0.3)
     B=&FloatFromCurve(FloatCurve=DataInterface() {propertyValues {Curve {keys [
          {interpMode=RCIM_Cubic, tangentMode=RCTM_User, time=0.1, value=0}
          {interpMode=RCIM_Cubic, tangentMode=RCTM_User, time=0.3, value=0.8}
          {tangentMode=RCTM_User, time=0.8, value=0}
         ]}}})
    )
    Scale RGB=&MakeVectorFromLinearColorRGB(Color=<rgba 1 1 1 1>)
    ScaleA=true
    ScaleRGB=true
   )
   /SetVariables(Particles.Color002=&MultiplyLinearColor(
     Linear Color=&MakeLinearColorFromVectorAndFloat(
      "Vector (RGB)"=&MakeVectorFromLinearColorRGB(Color=<rgba 1 1 1 1>)
      "Float (Alpha)"=&FN_FrontEndFloatModifier(Gameplay=1.2, Front End=0.5)
     )
     Scale Factor=&MakeVector4_FromLinearColor(LinearColor=$Particles.Color)
    )) {isSetParameters=true}
   /SetVariables(Particles.Color_triangle:LinearColor=&MakeLinearColorFromVectorAndFloat(
     "Vector (RGB)"=&MakeVectorFromLinearColorRGB(Color=&MakeCustomLinearColorFromBool(
       Bool=&RandomBool()
       True Color=<rgba 0 0.65 0.1934 1>
       False Color=&MakeCustomLinearColorFromBool(Bool=&RandomBool(), True Color=<rgba 0.4456 1 0 1>, False Color=<rgba 0 0.4818 0.7188 1>)
      ))
     "Float (Alpha)"=1
    )) {isSetParameters=true}
   /SetVariables(Particles.Color_triangleglow:LinearColor=&MultiplyLinearColor(
     Linear Color=$Particles.Color_triangle
     Scale Factor=&MakeVector4_FromLinearColor(LinearColor=/Niagara/DynamicInputs/TypeConversions/MakeLinearColorFromFloat(Value:NiagaraFloat=0.5))
    )) {isSetParameters=true}
  ]
  rendererClasses [&NiagaraSpriteRendererProperties]
  renderers [
   {rendererClass=&NiagaraSpriteRendererProperties}
   {rendererClass=&NiagaraSpriteRendererProperties, rendererIndex=1}
   {rendererClass=&NiagaraSpriteRendererProperties, rendererIndex=2}
   {rendererClass=&NiagaraSpriteRendererProperties, rendererIndex=3}
  ]
 }
 Trail_Sparkles {
  simTarget=GPUComputeSim
  emitterUpdateScript [
   &EmitterState(Life Cycle Mode=::NewEnumerator1, Loop Delay=1.15, Loop Duration=1, UseLoopDelay=false)
   &SpawnRate(
    Spawn Probability=&MakeCustomFloatFromBool(
     Bool=$User.bIsFrontEndPreview
     True Float=1
     False Float=&Lerp_Float(
      A=&Lerp_Float(A=0.4, B=1, Alpha=$System.InterpolateOverTime001.InterpolateOverTime_Float.Moving Average)
      B=1
      Alpha=$System.InterpolateOverTime.InterpolateOverTime_Float.Moving Average
     )
    )
    SpawnRate=100
    Use Spawn Probability=true
   )
  ]
  particleSpawnScript [
   &PartitionParticles()
   &InitializeParticle(
    Color=<rgba 0.6027 1 0 1>
    Lifetime Max=0.6
    Lifetime Min=0.2
    Lifetime Mode=::NewEnumerator1
    Sprite Size Mode=::NewEnumerator1
    Uniform Sprite Size Max=7
    UsePositionOffset=false
    Write Material Random=true
   )
   &ShapeLocation(
    Cylinder Height=130
    Cylinder Radius=20
    MaskSpawnGroup=false
    Non Uniform Scale=<1 1 0.5>
    Offset Mode=::NewEnumerator0
    Override Seed=false
    Rotation Axis=<1 0.5 0.4>
    Rotation Mode=::NewEnumerator1
    Shape Primitive=::NewEnumerator1
    Surface Only=false
    Surface Only Band Thickness=0.1
    UseOwnerScale=true
   )
   /SetVariables(Particles.RandomCore:NiagaraFloat=&FloatFromCurve(
     FloatCurve=DataInterface() {propertyValues {Curve {keys [
         {arriveTangent=0.0559, interpMode=RCIM_Cubic, leaveTangent=0.0559, tangentMode=RCTM_User, time=0, value=0.2}
         {arriveTangent=0.1895, interpMode=RCIM_Cubic, leaveTangent=0.1895, tangentMode=RCTM_User, time=0.9, value=0.25}
         {arriveTangent=1.1808, interpMode=RCIM_Cubic, leaveTangent=1.1808, tangentMode=RCTM_User, time=1, value=0.3}
        ]}}}
     CurveIndex=&RandomRangeFloat(Minimum=0, Maximum=1)
     Scale Curve=2
    )) {isSetParameters=true}
  ]
  particleUpdateScript [
   &ParticleState()
   &SkeletalMeshLocation(
    Bone / Socket Index=&SelectIntFromArray(Direct Array Index=$Particles.PartitionParticles.Partition, Int Selection Array=DataInterface() {propertyValues {IntData [0,1]}})
    MaskSpawnGroup=false
    Mesh Sampling Type=::NewEnumerator5
    Sampled Position Offset=&AddVectorToPosition(
     Position=Linked() {linkedVariable {name=Particles.Initial.ShapeLocation.ShapeVector, type=Vector3f}}
     Vector=&Lerp_Vector(
      A=&SelectVectorFromArray(
       Direct Array Index=$Particles.PartitionParticles.Partition
       Vector Selection Array=DataInterface() {propertyValues {FloatData [<40 20 15>, <-40 -20 -15>]}}
      )
      B=&SelectVectorFromArray(
       Direct Array Index=$Particles.PartitionParticles.Partition
       Vector Selection Array=DataInterface() {propertyValues {FloatData [<40 120 15>, <-40 -120 -15>]}}
      )
      Alpha=&FloatFromCurve(
       FloatCurve=DataInterface() {propertyValues {Curve {keys [<at 0 0>,<at 1 1>]}}}
       Scale Curve=&MakeCustomFloatFromBool(
        Bool=$User.bIsFrontEndPreview
        True Float=2
        False Float=&Lerp_Float(
         A=&Lerp_Float(A=0.5, B=2, Alpha=$System.InterpolateOverTime001.InterpolateOverTime_Float.Moving Average)
         B=3
         Alpha=$System.InterpolateOverTime.InterpolateOverTime_Float.Moving Average
        )
       )
      )
     )
    )
    Skeletal Mesh=$System.Skeletal Mesh Reader
    Socket Sampling Mode=::NewEnumerator5
   )
   &ScaleSpriteSize(Scale Factor=&Vector2DFromFloat(Value=&FloatFromCurve(
      FloatCurve=DataInterface() {propertyValues {Curve {keys [
          <at 0 1.2 RCIM_Cubic>
          {arriveTangent=-1.3, interpMode=RCIM_Cubic, leaveTangent=-1.3, time=0.3, value=1}
          {arriveTangent=-2.4232, interpMode=RCIM_Cubic, leaveTangent=-2.4232, tangentMode=RCTM_User, time=0.5, value=0.55}
          {arriveTangent=0.0448, interpMode=RCIM_Cubic, leaveTangent=0.0448, tangentMode=RCTM_User, time=1, value=0}
         ]}}}
      Scale Curve=&FN_FrontEndFloatModifier(Gameplay=0.35, Front End=0.2)
     )))
   &ScaleColor(
    Scale Alpha=&FloatFromCurve(FloatCurve=DataInterface() {propertyValues {Curve {keys [<at 0 1>,<at 1 0>]}}}, Scale Curve=0.7)
    Scale RGB=&Multiply_VectorByFloat(Vector=<1 1 1>, Float=1)
    ScaleA=true
    ScaleRGB=true
   )
   /SetVariables(Particles.Glow03_Color=&MakeLinearColorFromVectorAndFloat(
     "Vector (RGB)"=&MakeVectorFromLinearColorRGB(Color=&SelectLinearColorFromArray(
       Array Sampling Mode=::NewEnumerator1
       Direct Array Index=&MakeIntFromFloat(Float=$User.VariantSwitch)
       Color Selection Array=DataInterface() {propertyValues {ColorData [<rgba 0 1 0.1402 0.2>, <rgba 0.3654 1 0.0185 0.2>]}}
       Random Seed=?
      ))
     "Float (Alpha)"=&FloatFromCurve(
      FloatCurve=DataInterface() {propertyValues {Curve {keys [<at 0 1>,<at 1 0>]}}}
      Scale Curve=&Multiply_Float(
       A=&FN_FrontEndFloatModifier(Gameplay=0.25, Front End=0.15)
       B=&FNTimeOfDayScalar(Night Scale=0.5, Morning Scale=0.8, Evening Scale=0.8)
      )
     )
    )) {isSetParameters=true}
   /SetVariables(Particles.Glow03_Size=&Multiply_Vector2D_ByFloat(
     Float=&FloatFromCurve(
      FloatCurve=DataInterface() {propertyValues {Curve {keys [
          <at 0 0 RCIM_Cubic>
          <at 0.2 1.3 RCIM_Cubic>
          {arriveTangent=-0.1423, interpMode=RCIM_Cubic, leaveTangent=-0.1423, tangentMode=RCTM_User, time=1, value=0}
         ]}}}
      Scale Curve=18
     )
     Vector2D=<1 1>
    )) {isSetParameters=true}
   &CurlNoiseForce(Bias Noise=false, Noise Strength=500, Pan Noise=false)
   &SolveForcesAndVelocity(Clamp Velocity=false, Limit Acceleration=false)
  ]
  rendererClasses [&NiagaraSpriteRendererProperties]
  renderers [
   {rendererClass=&NiagaraSpriteRendererProperties}
   {rendererClass=&NiagaraSpriteRendererProperties, rendererIndex=1}
  ]
 }
}
summary {
 systemName=NS_Glider_LocalTalent_Flare_Idle
 emitters [
  {emitterName=Sprites, rendererClasses [&NiagaraSpriteRendererProperties]}
  {emitterName=SpeedLine, simTarget=GPUComputeSim, rendererClasses [&NiagaraSpriteRendererProperties]}
  {emitterName=Trail_Sparkles, simTarget=GPUComputeSim, rendererClasses [&NiagaraSpriteRendererProperties]}
  {emitterName=SoftTrail, rendererClasses [&NiagaraRibbonRendererProperties]}
  {emitterName=MainTrail002, rendererClasses [&NiagaraRibbonRendererProperties]}
 ]
 userVariables [
  User.bIsFrontEnd {type=NiagaraBool}
  User.bIsFrontEndPreview {type=NiagaraBool}
  User.bIsFullyDeployed {type=NiagaraBool}
  User.ForwardDot {type=NiagaraFloat, defaultValue=0, description=""}
  User.ParaGlideLeanAlpha {type=NiagaraFloat, defaultValue=0, description=""}
  User.RibbonIDOffset {type=NiagaraInt32}
  User.VariantSwitch {type=NiagaraFloat, defaultValue=0, description=""}
 ]
}
